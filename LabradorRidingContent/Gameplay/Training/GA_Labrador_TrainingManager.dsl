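; Gameplay ability that manages taming progress for a rideable Labrador AI pawn while a player rides it.
; Riding-tag waits track the rider, and a server movement callback runs a training timer during movement, pausing it after brief idleness.
; Progress drives a rider widget event and pitch-shifted audio.
; Full training applies a buff, ends the waits, and tosses marked loot plus gold; separate events handle a delayed loot-and-kill sequence.
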(event Ability|EventActivateAbility
  (bind (_event_data _fort_pawn) (Setup|SetupGenericAbilityAnyPawn false)
    (:Output
      (bind _asbp_aipawn_labrador_ridable (Utilities|Casting|CastToBP_AIPawn_Labrador_Ridable _fort_pawn))
      (Variables|Default|SetLabradorPawn _asbp_aipawn_labrador_ridable)
      (CallFunction|WaitRidingBegin)
      (Utilities|IsServer
        (:true
          (Character|BindEventtoOnCharacterMovementUpdated (Variables|Default|GetLabradorPawn) (EventDispatchers|CreateEvent :!event MovementUpdate)))))))

(event Custom|WaitRidingBegin
  ; Wait Events
  (bind _asynctaskproxy (Ability|Tasks|WaitGameplayTagAddToActor (Variables|Default|GetLabradorPawn) "(TagName=\"Riding.IsBeingRidden\")")
    (:then
      (Variables|Default|SetWaitStartRiding _asynctaskproxy))
    (:Added
      (CallFunction|CacheNewRider))))

(event Custom|WaitForRidingEnd
  ; Wait Events
  (bind _asynctaskproxy (Ability|Tasks|WaitGameplayTagRemoveFromActor (Variables|Default|GetLabradorPawn) "(TagName=\"Riding.IsBeingRidden\")")
    (:then
      (Variables|Default|SetWaitStopRiding _asynctaskproxy))
    (:Removed
      (CallFunction|ClientToggleProgress false)
      (CallFunction|ToggleAutoTraining false))))

(event Custom|ClientToggleProgress
  (params (Show :type Boolean))
  ; Client Widget/FX
  (GameplayEventRouter|BroadcastStatefulEvent (GameplayEventRouter|GetGlobalEventRouter) (Utilities|Struct|MakeStructLabradorRidingWidgetProgress (Utilities|ResolveSoftReference (Variables|Default|GetLabradorPawn)) Show) (Variables|Default|GetRiderPawn))
  (if Show
    (bind _capturesound (Variables|Default|GetCaptureSound)
      (:then
        (Audio|Components|Audio|Play _capturesound 0.0)
        (Audio|Components|Audio|FadeIn _capturesound 1.0 1.0 0.0 "Linear" :!class /Script/Engine.AudioComponent))
      (:else
        (bind _returnvalue (Audio|SpawnSound2D "/Labrador/Sound/Foley/MSS_Labrador_Taming_Progression.MSS_Labrador_Taming_Progression" 2.0))
        (bind _output_get (Variables|Default|SetCaptureSound _returnvalue))
        (Audio|Components|Audio|Play _output_get 0.0)
        (Audio|Components|Audio|FadeIn _output_get 1.0 1.0 0.0 "Linear" :!class /Script/Engine.AudioComponent)))
    (else
      (Audio|Components|Audio|FadeOut (Variables|Default|GetCaptureSound) 1.0 0.0 "Linear" :!class /Script/Engine.AudioComponent))))

(event Custom|ToggleAutoTraining
  (params (Enabled :type Boolean))
  ; Server Auto Training Overtime
  (if Enabled
    (if (< (Variables|/Labrador/Pawn/BP_AIPawn_Labrador_Ridable.BP_AIPawn_Labrador_Ridable_C|GetTrainingPercentage (Variables|Default|GetLabradorPawn)) 1.0)
      (bind _returnvalue (Utilities|Time|SetTimerbyEvent (EventDispatchers|CreateEvent :!event IncreaseTrainingPercentage) (ScalableFloat|GetValueatLevel (Variables|Default|GetHF_TrainingOvertimeRate)) true true))
      (Variables|Default|SetHandle_TrainOvertime _returnvalue)
      (CallFunction|ClientToggleProgress true))
    (else
      (Utilities|Time|ClearandInvalidateTimerbyHandle (Variables|Default|GetHandle_TrainOvertime)))))

(event Custom|ClientPercentUpdate
  (params (UpdatedProgress :type Float%20%28double-precision%29))
  ; Client Widget/FX
  (Audio|Parameter|SetFloatParameter (Variables|Default|GetCaptureSound) "PitchShift" UpdatedProgress :!class /Script/Engine.AudioComponent))

(event Custom|SleepProgress
  ; Client Widget/FX
  (CallFunction|ToggleAutoTraining false)
  (CallFunction|ClientToggleProgress false))

(event Custom|LevelUp
  ; LevelUp
  (GameplayEffects|ApplyGameplayEffectToSelf :self (Ability|GetFortAbilitySystemComponent (Variables|Default|GetLabradorPawn)) :GameplayEffectClass "/LabradorRidingContent/Gameplay/Training/GE_LabradorRiding_Training_Buff.GE_LabradorRiding_Training_Buff_C" :Level 0.0)
  (Ability|Async|EndAction (Variables|Default|GetWaitStartRiding))
  (Ability|Async|EndAction (Variables|Default|GetWaitStopRiding))
  (Default|DropLegendaryLoot))

(event Custom|AnticipateKillLabrador
  ; Kill
  (Class|BPAIPawnLabradorRidable|ToggleAnticipationFX (Variables|Default|GetLabradorPawn) true)
  (Utilities|Time|SetTimerbyEvent (EventDispatchers|CreateEvent :!event DropLegendaryLoot) 9.5)
  (Utilities|Time|SetTimerbyEvent (EventDispatchers|CreateEvent :!event Custom|FinalizeKillLabrador) 1e+01))

(event Custom|Client_LeakAnim
  ; Kill
  (bind _returnvalue (Class|FortAnimInputEvent|CreateAnimInputEvent "aie.Labrador.Leak"))
  (Animation|TriggerAnimInputEvent (Variables|Default|GetLabradorPawn) _returnvalue))

(event Custom|FinalizeKillLabrador
  ; Kill
  (Class|BPAIPawnLabradorRidable|ToggleAnticipationFX (Variables|Default|GetLabradorPawn) false)
  (Game|Damage|ApplyDamage (Variables|Default|GetLabradorPawn) 1e+03))

(fn CacheNewRider
  (bind _asfort_pawn (Utilities|Casting|CastToFortPawn (Components|GetOwner (Variables|/Script/RidingCodeRuntime.RidableComponent|GetDoNotUse_PrimaryRider (Actor|GetComponentbyClass (Variables|Default|GetLabradorPawn) "/LabradorRidingContent/Components/LabradorRiding_LabradorRidableComponent.LabradorRiding_LabradorRidableComponent_C")))))
  (Variables|Default|SetRiderPawn _asfort_pawn)
  (CallFunction|WaitforRidingEnd)
  (return))

(fn CalculateToss
  (bind _riderpawn (Variables|Default|GetRiderPawn))
  (Utilities|IsValid _riderpawn
    (:Is%20Valid
      (return (+ (* (Math|Vector|GetForwardVector (Camera|GetCameraRotation (Variables|/Script/Engine.PlayerController|GetPlayerCameraManager (Fortnite|GetFortPlayerControllerfromActor _riderpawn)))) 1e+03) "0.000000,0.000000,700.000000")))
    (:Is%20Not%20Valid
      (return (+ (* (Transformation|GetActorForwardVector (Variables|Default|GetLabradorPawn)) 1e+03) "0.000000,0.000000,700.000000")))))

(fn DropLegendaryLoot
  (Utilities|IsValid (Variables|Default|GetLabradorPawn)
    (:Is%20Valid
      (Utilities|FlowControl|Sequence
        (:then_0
          (bind _returnvalue (Inventory|SpawnPickupInWorldWithLootTier (Variables|Default|GetLoot%20Tier%20Name) (+ (Transformation|GetActorLocation (Variables|Default|GetLabradorPawn)) "0.000000,0.000000,145.000000") 0 0 true true (CallFunction|CalculateToss) "Tossed" "Unset" (Loot|GetLootLevel))) ; use when loot tier added
          (if (Utilities|IsArrayNonEmpty _returnvalue)
            (MarkerComponent|MarkActoronServer (Actor|GetComponentbyClass (Fortnite|GetFortPlayerControllerfromActor (Variables|Default|GetRiderPawn)) "/Script/FortniteGame.AthenaMarkerComponent") (Utilities|Array|Get(acopy) _returnvalue) false true)
            (Ability|ExecuteGameplayCueWithParamsOnOwner (Variables|Default|GetGameplayCueLeak_Legend) (Ability|GameplayCue|MakeGameplayCueParameters :NormalizedMagnitude 0.0 :RawMagnitude 0.0 :Location "0,0,0" :Normal "0,0,0" :SourceObject (Utilities|Array|Get(acopy) _returnvalue) :GameplayEffectLevel 1 :AbilityLevel 1 :bReplicateLocationWhenUsingMinimalRepProxy false))
            (CallFunction|ClientLeakAnim)))
        (:then_1
          (Default|DropGold (Variables|Default|GetLabradorPawn)))))))

(fn IncreaseTrainingPercentage
  (bind _output_get (Variables|/Labrador/Pawn/BP_AIPawn_Labrador_Ridable.BP_AIPawn_Labrador_Ridable_C|SetTrainingPercentage (Math|Float|Clamp(Float) (+ (Variables|/Labrador/Pawn/BP_AIPawn_Labrador_Ridable.BP_AIPawn_Labrador_Ridable_C|GetTrainingPercentage (Variables|Default|GetLabradorPawn)) (ScalableFloat|GetValueatLevel (Variables|Default|GetHF_TrainingOvertimeAddPerTick)))) (Variables|Default|GetLabradorPawn)))
  (CallFunction|ClientPercentUpdate _output_get)
  (if (>= _output_get 1.0)
    (CallFunction|ToggleAutoTraining false)
    (CallFunction|ClientToggleProgress false)
    (CallFunction|LevelUp)))

(fn MovementUpdate
  (params (DeltaSeconds :type Float%20%28single-precision%29) (OldLocation :type Vector) (OldVelocity :type Vector))
  (if (GameplayTags|HasMatchingGameplayTag (Variables|Default|GetLabradorPawn) "(TagName=\"Riding.IsBeingRidden\")")
    (if (> (Math|Vector|VectorLength OldVelocity) 1e+02)
      ; Start Training when movement begins
      (Utilities|FlowControl|Sequence
        (:then_0
          (if (Utilities|Time|IsTimerActivebyHandle (Variables|Default|GetHandle_SleepTraining))
            (Utilities|Time|ClearandInvalidateTimerbyHandle (Variables|Default|GetHandle_SleepTraining))))
        (:then_1
          (if (not (Utilities|Time|IsTimerActivebyHandle (Variables|Default|GetHandle_TrainOvertime)))
            (CallFunction|ToggleAutoTraining true))))
      (elif (not (Utilities|Time|IsTimerActivebyHandle (Variables|Default|GetHandle_SleepTraining)))
        ; Stop Training when not moving for duration
        (if (Utilities|Time|IsTimerActivebyHandle (Variables|Default|GetHandle_TrainOvertime))
          (bind _returnvalue (Utilities|Time|SetTimerbyEvent (EventDispatchers|CreateEvent :!event SleepProgress) 1.5))
          (Variables|Default|SetHandle_SleepTraining _returnvalue))))))
